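# Niagara system for the Sesameseed backpack cosmetic that renders slowly circling sprite clouds in two stacked rings.
# One burst of looping particles splits into two partitions, with the lower ring smaller and rotating faster.
# Per-index curves shape cloud height, size, alpha and material parameters, giving each ring an uneven silhouette.
# Color follows time of day, with front-end and quality-level overrides, and two sprite renderers draw the particles.
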NiagaraSystem /BRCosmetics/Accessories/FORT_Backpacks/Backpack_M_MED_SesameSeed/FX/NS_Backpack_M_Sesameseed_Clouds
dependencies {
 usedDataInterfaces [&NiagaraDataInterfaceCurve, &NiagaraDataInterfaceArrayFloat, &NiagaraDataInterfaceArrayFloat3]
 usedDynamicInputs []
 usedModules [&SystemState]
}
emitters {Clouds {
  emitterUpdateScript [
   &EmitterState(Life Cycle Mode=::NewEnumerator1, Loop Behavior=::NewEnumerator1, Loop Duration Mode=::NewEnumerator1, UseLoopDelay=false)
   &SpawnBurst_Instantaneous(Spawn Count=32, Use Spawn Probability=false, UseLoopCountLimit=false)
   /SetVariables(Emitter.Emissive:LinearColor=&MakeCustomLinearColorFromBool(
     Bool=&SetBoolByFloatComparison(A=1, B=&FN_FrontEndFloatModifier(Gameplay=0, Front End=1), Comparison Type=::NewEnumerator4)
     True Color=<rgba 0.05 0.05 0.05 1>
     False Color=&FNTimeOfDayColor(
      Night Color=<rgba 0.0286 0.0697 0.25 1>
      Morning Color=<rgba 0.0986 0.0712 0.1 1>
      Day Color=<rgba 0.098 0.149 0.2 1>
      Evening Color=<rgba 0.0994 0.0609 0.1 1>
     )
    )) {isSetParameters=true}
  ]
  particleSpawnScript [
   &PartitionParticles(Particles Per Partition=16, Partition Mode:ENiagara_PartitionMode=::NewEnumerator1)
   /SetVariables(
    Particles.CloudsCurve:NiagaraFloat=&FloatFromCurve(
     FloatCurve=DataInterface() {propertyValues {Curve {keys [
         {interpMode=RCIM_Cubic, tangentMode=RCTM_User, time=0, value=0.625}
         {arriveTangent=-0.5136, interpMode=RCIM_Cubic, leaveTangent=-0.5136, time=0.113, value=0.6735}
         {arriveTangent=0.3831, interpMode=RCIM_Cubic, leaveTangent=0.3831, time=0.2386, value=0.5025}
         {interpMode=RCIM_Cubic, tangentMode=RCTM_User, time=0.4431, value=0.8}
         {arriveTangent=-0.0078, interpMode=RCIM_Cubic, leaveTangent=-0.0078, time=0.6, value=0.55}
         {arriveTangent=-0.5625, interpMode=RCIM_Cubic, leaveTangent=-0.5625, time=0.8591, value=0.7968}
         {tangentMode=RCTM_User, time=1, value=0.325}
        ]}}}
     CurveIndex=$Particles.PartitionParticles.Normalized Index In Partition
    )
    Particles.LowerCloudsCurve:NiagaraFloat=&FloatFromCurve(
     FloatCurve=DataInterface() {propertyValues {Curve {keys [
         {arriveTangent=0.5776, interpMode=RCIM_Cubic, leaveTangent=0.5776, tangentMode=RCTM_User, time=-0.0167, value=0.1059}
         {arriveTangent=1.4595, interpMode=RCIM_Cubic, leaveTangent=1.4595, time=0.1, value=0.65}
         {arriveTangent=0.1667, interpMode=RCIM_Cubic, leaveTangent=0.1667, time=0.1945, value=0.4141}
         {arriveTangent=0.5776, interpMode=RCIM_Cubic, leaveTangent=0.5776, tangentMode=RCTM_User, time=0.4, value=0.7}
         {arriveTangent=0.3388, interpMode=RCIM_Cubic, leaveTangent=0.3388, tangentMode=RCTM_User, time=0.5725, value=0.4641}
         {arriveTangent=0.4703, interpMode=RCIM_Cubic, leaveTangent=0.4703, tangentMode=RCTM_User, time=0.8701, value=0.6824}
         {tangentMode=RCTM_User, time=0.9944, value=0.3412}
        ]}}}
     CurveIndex=$Particles.PartitionParticles.Normalized Index In Partition
    )
   ) {isSetParameters=true}
   &InitializeParticle(
    Color=<rgba 0.4162 0.6357 0.75 1>
    Sprite Rotation Angle=&RandomRangeFloat(Minimum=-0.1, Maximum=0.1)
    Sprite Rotation Mode=::NewEnumerator1
    Sprite Size Mode=::NewEnumerator3
    Uniform Sprite Size=&Multiply_Float(
     A=5.5
     B=&SelectFloatFromArray(Direct Array Index=$Particles.PartitionParticles.Partition, Float Selection Array=DataInterface() {propertyValues {FloatData [1,0.5]}})
    )
    UsePositionOffset=false
    Write Material Random=true
   )
   &SystemLocation(
    MaskSpawnGroup=false
    Offset=&SelectVectorFromArray(
     Direct Array Index=$Particles.PartitionParticles.Partition
     Vector Selection Array=DataInterface() {propertyValues {FloatData [<0 0 -1.5>, <0 0 -4.5>]}}
    )
   )
   /SetVariables(Particles.Position=&AddVectorToPosition(
     Position=$Particles.Position
     Vector=&TransformVector(
      Vector=&MakeVector(
       X=0
       Y=0
       Z=&Multiply_Float(
        A=/Niagara/DynamicInputs/Math/CalculateParticleRadius(
         Radius Calculation Type:ENiagaraCollisionRadiusOptions=::NewEnumerator0
         Particle Radius:NiagaraFloat=1
         Mesh Dimensions:Vector3f=<1 1 1>
         Method for Calculating Particle Radius:ENiagaraCalculateRadiusOptions=::NewEnumerator1
        )
        B=&Multiply_Float(A=0.5, B=$Particles.CloudsCurve)
       )
      )
      Source Space=::Simulation
      Destination Space=::Local
      Apply Scale=true
      Transform Vector as Position=true
     )
    )) {isSetParameters=true}
   &ShapeLocation(
    Cylinder Distribution=::NewEnumerator2
    Cylinder Height=0
    Cylinder Radius=&SelectFloatFromArray(Direct Array Index=$Particles.PartitionParticles.Partition, Float Selection Array=DataInterface() {propertyValues {FloatData [5,2.5]}})
    MaskSpawnGroup=false
    Offset:NiagaraPosition=$Particles.Position
    Offset Coordinate Space=::Local
    Offset Mode=::NewEnumerator0
    Rotation Angle:NiagaraFloat=&FloatFromCurve(FloatCurve=DataInterface() {propertyValues {Curve {keys [<at 0 1>,<at 1 0>]}}})
    Rotation Axis=<0 0 1>
    Rotation Mode=::NewEnumerator1
    Shape Primitive=::NewEnumerator1
    U Position=&FloatFromCurve(
     FloatCurve=DataInterface() {propertyValues {Curve {keys [{tangentMode=RCTM_User, time=0, value=0}, {tangentMode=RCTM_User, time=1, value=1}]}}}
     CurveIndex=$Particles.PartitionParticles.Normalized Index In Partition
    )
    UseOwnerScale=true
   )
   &CameraOffset(Camera Offset Amount=2)
  ]
  particleUpdateScript [
   &ParticleState(
    Kill Particles When Lifetime Has Elapsed=false
    Let Infinitely Lived Particles Die When Emitter Deactivates=true
    Loop Particles Lifetime=true
   )
   &ScaleSpriteSize(
    Scale Sprite Size Mode=::NewEnumerator0
    Uniform Scale Factor=&MaskFloatBySpawnGroup(
     Masked Float=&Multiply_Float(A=1, B=$Particles.CloudsCurve)
     Passthrough Float=&Multiply_Float(A=1, B=$Particles.LowerCloudsCurve)
     Spawn Group Mask=$Particles.PartitionParticles.Partition
    )
   )
   &ScaleColor(
    Scale Alpha=&MaskFloatBySpawnGroup(
     Masked Float=&Lerp_Float(A=-0.5, B=1.5, Alpha=$Particles.CloudsCurve)
     Passthrough Float=&Lerp_Float(A=-0.5, B=1.5, Alpha=$Particles.LowerCloudsCurve)
     Spawn Group Mask=$Particles.PartitionParticles.Partition
    )
    Scale RGB=&MakeVectorFromLinearColorRGB(Color=&MakeCustomLinearColorFromBool(
      Bool=&SetBoolByFloatComparison(A=1, B=&FN_FrontEndFloatModifier(Gameplay=0, Front End=1), Comparison Type=::NewEnumerator4)
      True Color=<rgba 0.75 0.75 0.75 1>
      False Color=&MakeCustomLinearColorFromBool(
       Bool=&SetBoolByFloatComparison(
        A=&MakeFloatFromInt(Int=2)
        B=&MakeFloatFromInt(Int=$Engine.QualityLevel)
       )
       True Color=<rgba 1 1 1 1>
       False Color=&FNTimeOfDayColor(
        Night Color=<rgba 0.3587 0.3852 0.6 1>
        Morning Color=<rgba 0.8 0.4208 0.5 1>
        Day Color=<rgba 1 0.4849 0.3748 1>
        Evening Color=<rgba 0.8 0.3232 0.363 1>
       )
      )
     ))
    ScaleA=true
    ScaleRGB=true
   )
   /SetVariables(Particles.Position=/Niagara/DynamicInputs/Vectors/RotateVector(
     Vector To Rotate:Vector3f=Linked() {linkedVariable {name=Particles.Position, type=NiagaraPosition}}
     Yaw:NiagaraFloat=&SelectFloatFromArray(Direct Array Index=$Particles.PartitionParticles.Partition, Float Selection Array=DataInterface() {propertyValues {FloatData [9,18]}})
     Pitch:NiagaraFloat=0
     Roll:NiagaraFloat=0
     Rotation Rate:NiagaraFloat=$Engine.DeltaTime
    )) {isSetParameters=true}
   &DynamicMaterialParameters(
    Index 0 Param 1=&RandomRangeFloat(Minimum=0, Maximum=1)
    Index 2 Param 1=&MaskFloatBySpawnGroup(
     Masked Float=&Lerp_Float(A=3.819, B=-2, Alpha=$Particles.CloudsCurve)
     Passthrough Float=&Lerp_Float(A=3.9592, B=-2, Alpha=$Particles.LowerCloudsCurve)
     Spawn Group Mask=$Particles.PartitionParticles.Partition
    )
    Index 2 Param 2=&Add_Float(
     A=&Multiply_Float(A=$Engine.RealTime, B=&RandomRangeFloat(Minimum=0.05, Maximum=0.1))
     B=&RandomRangeFloat(Minimum=0, Maximum=1)
    )
    Index 2 Param 3=2
    Index 2 Param 4=&Multiply_Float(
     A=&RandomRangeFloat(Minimum=1, Maximum=1.1)
     B=&SelectFloatFromArray(Direct Array Index=$Particles.PartitionParticles.Partition, Float Selection Array=DataInterface() {propertyValues {FloatData [0.5,0.4]}})
    )
    Module.Param0WriteEnabled=true
    Module.Param0WriteEnabled2=true
    Module.Param1WriteEnabled=true
    Module.Param1WriteEnabled2=true
    Module.Param2WriteEnabled=true
    Module.Param2WriteEnabled2=true
    Module.Param3WriteEnabled=true
    Module.Param3WriteEnabled2=true
    Write Parameter Index 2=true
   )
  ]
  rendererClasses [&NiagaraSpriteRendererProperties]
  renderers [
   {rendererClass=&NiagaraSpriteRendererProperties}
   {rendererClass=&NiagaraSpriteRendererProperties, rendererIndex=1}
  ]
 }}
summary {
 systemName=NS_Backpack_M_Sesameseed_Clouds
 emitters [{emitterName=Clouds, rendererClasses [&NiagaraSpriteRendererProperties]}]
 userVariables [
  User.bIsFrontEnd {type=NiagaraBool}
  User.bIsFrontEndPreview {type=NiagaraBool}
 ]
}
