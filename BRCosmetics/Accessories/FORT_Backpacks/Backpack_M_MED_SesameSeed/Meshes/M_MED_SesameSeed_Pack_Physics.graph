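# Physics asset for the M_MED_SesameSeed_Pack backpack cosmetic, giving its skeletal mesh simple collision and simulation bodies.
# A large capsule covers the pack bone and a tiny capsule sits on the root bone.
# A single constraint joins pack to root with tight swing and twist limits, so any physical movement of the pack relative to the root stays slight.
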PhysicsAsset /BRCosmetics/Accessories/FORT_Backpacks/Backpack_M_MED_SesameSeed/Meshes/M_MED_SesameSeed_Pack_Physics
bodies {
 pack [{shapeType=Capsule, length=16.6654, radius=13.2995, shapeName=pack_capsule, center=<0 -10.8412 -0.9948>, rotation=<rot 3.0233 -90 -1.0121>}]
 root [{shapeType=Capsule, length=0.505, radius=0.505, shapeName=root_capsule, rotation=<rot 0 -90 0>}]
}
constraints [{bone1Name=pack, bone2Name=root, swing1LimitDegrees=1, swing2LimitDegrees=4, twistLimitDegrees=2.4}]
