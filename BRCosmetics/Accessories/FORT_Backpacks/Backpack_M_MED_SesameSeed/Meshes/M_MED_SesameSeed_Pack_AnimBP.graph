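# Animation Blueprint for the SesameSeed backpack cosmetic, adding physics secondary motion and a gentle floating drift to the pack.
# A partially blended RigidBody node simulates the pack's physics asset in root-bone space with overridden gravity and scaled-down component motion.
# Two additive bone-space modifiers on the pack bone then apply a fixed offset and a time-varying translation.
# The event graph builds that translation each update from three real-time sine waves at different frequencies and amplitudes.
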AnimBlueprint /BRCosmetics/Accessories/FORT_Backpacks/Backpack_M_MED_SesameSeed/Meshes/M_MED_SesameSeed_Pack_AnimBP
target_skeleton=&/M_MED_SesameSeed_Pack_Skeleton
graphs [
 AnimGraph {nodes [
   n1 = RigidBody(ComponentPose=Local To Component(LocalPose=Input Pose()), Alpha=0.4) {
    bOverrideWorldGravity=true
    overridePhysicsAsset=&/M_MED_SesameSeed_Pack_Physics
    simulationSpace=BaseBoneSpace
    alphaBoolBlend {blendOption=Linear}
    baseBoneRef {boneName=root}
    componentAppliedLinearAccClamp=<5000 5000 5000>
    componentLinearAccScale=<0.4 0.4 0.4>
    componentLinearVelScale=<0.4 0.4 0.4>
    overrideWorldGravity {z=-980}
    simSpaceSettings {
     maxAngularAcceleration=30
     maxAngularVelocity=1
     maxLinearAcceleration=30
     maxLinearVelocity=10
     velocityScaleZ=2
     worldAlpha=1
     externalLinearDragV=<0.3 0.3 0.3>
     externalLinearVelocity=<0.1 0.1 0.1>
    }
   }
   n2 = "Transform (Modify) Bone - Bone: pack"(ComponentPose=@n1, Translation="0,-4.2,0", Rotation="0.063,0.011,-2.525") {
    rotationMode=BMM_Additive
    rotationSpace=BCS_BoneSpace
    translationMode=BMM_Additive
    translationSpace=BCS_BoneSpace
    alphaBoolBlend {blendOption=Linear}
    alphaScaleBiasClamp {bInterpResult=true, interpSpeedDecreasing=10000, interpSpeedIncreasing=10000}
    boneToModify {boneName=pack}
    rotation=<rot 0.063 0.011 -2.525>
    translation {y=-4.2}
   }
   n3 = "Transform (Modify) Bone - Bone: pack"(ComponentPose=@n2, Translation=Get dyn_pack_float() -> (dyn_pack_float:Vector), Rotation="0,0,0") {
    rotationMode=BMM_Additive
    rotationSpace=BCS_BoneSpace
    translationMode=BMM_Additive
    translationSpace=BCS_BoneSpace
    alphaBoolBlend {blendOption=Linear}
    alphaScaleBiasClamp {bInterpResult=true, interpSpeedDecreasing=10000, interpSpeedIncreasing=10000}
    boneToModify {boneName=pack}
   }
   Output Pose(Result=Component To Local(ComponentPose=@n3)) {layerGroup=None}
  ]}
 EventGraph {nodes [
   Event BlueprintUpdateAnimation()
   n1 = GetRealTimeSeconds()
   n2 = "Sin (Radians)"(A=float * float(A=1, B=@n1))
   n3 = "Sin (Radians)"(A=float * float(A=0.4, B=@n1))
   n4 = "Sin (Radians)"(A=float * float(A=1.2, B=@n1))
   > Set dyn_pack_float(dyn_pack_float:Vector=MakeVector(X=float * float(A=@n3, B=1), Y=float * float(A=@n2, B=0.56), Z=float * float(A=@n4, B=2))) -> (Output_Get:Vector)
   TryGetPawnOwner()
  ]}
]
variables [dyn_pack_float {type=Vector}]
